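; Photo Mode gameplay ability that opens the emote wheel from the camera.
; On activation it calls the parent, finds the player pawn animation component on the avatar actor and refreshes its emote list.
; It then shows the "EmotePicker" radial menu through the local player's Fort picker context and ends immediately.
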(event Ability|EventActivateAbility
  (|CallParentFunction)
  (Emotes|RefreshEmotes (Actor|GetComponentbyClass (Ability|GetAvatarActorfromActorInfo) "/PhotoMode/Blueprints/B_PM_PlayerPawnAnim_Component.B_PM_PlayerPawnAnim_Component_C"))
  (Action|ShowRadialPickerMenu (PlayerController|LocalPlayerSubsystems|GetFortPickerContext (Ability|GetPlayerControllerfromActorInfo)) "EmotePicker" -1 0)
  (Ability|EndAbility))
