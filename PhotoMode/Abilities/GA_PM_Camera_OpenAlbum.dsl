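; Photo mode gameplay ability that, going by its name, opens the camera album; its activation event is empty in this graph.
; When the ability ends, it takes the stored album widget if present, unbinds all events from its OnWidgetDeactivated, deactivates it and clears the reference.
; This cleanup runs without checking whether the ability was cancelled.
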(event Ability|EventActivateAbility)

(event Ability|EventOnEndAbility
  (params (bWasCancelled :type Boolean))
  (bind _albumwidget (Variables|Default|GetAlbumWidget)
    (:then
      (Events|UnbindallEventsfromOnWidgetDeactivated _albumwidget)
      (ActivatableWidget|DeactivateWidget _albumwidget)
      (Variables|Default|SetAlbumWidget 0))
    (:else)))
