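; Base Photo Mode gameplay ability that cycles the photo camera orientation, meant to run only on non-authority clients.
; On authority it only prints a warning; otherwise it updates the player's PhotoConfig, sends an orientation gameplay event with the mapped tag, then ends.
; The next value comes from OrientationOrder, stepping by bForwardOrder with modulo wrap-around and falling back to the first entry if the current one is absent.
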(event Ability|EventActivateAbility
  (Ability|HasAuthority
    (:True
      (Development|PrintString "Warning: GA_Camera_ToggleOrientation should not be fired in Server"))
    (:False
      (bind _orientation (CallFunction|ChangeConfigtoNextOrientation))
      (Ability|SendGameplayEvent 0 "(TagName=\"CustomVerb.PhotoMode.Orientation\")" (Utilities|Struct|MakeGameplayEventData (Utilities|Map|Find (Variables|Default|GetOrientationTags) _orientation) (CallFunction|GetAvatarasFortPlayerPawn) (CallFunction|GetAvatarasFortPlayerPawn) 0 0 "()" "(GameplayTags=)" (GameplayTags|MakeGameplayTagContainerfromTag (Utilities|Map|Find (Variables|Default|GetOrientationTags) _orientation))))
      (Ability|EndAbility))))

(fn ChangeConfigToNextOrientation
  (bind _config (Class|BPhotoModeBPFL|GetPhotoConfig (Ability|GetPlayerControllerfromActorInfo)))
  (bind _output (CallFunction|GetNextOrientation (Variables|/Script/PhotoModeRuntime.PhotoConfig|GetOrientation _config)))
  (bind _output_get (Variables|/Script/PhotoModeRuntime.PhotoConfig|SetOrientation _output _config))
  (return _output_get))

(fn GetNextOrientation
  (params (CurrentOrientation :type EPhotoOrientation%20Enum :default "Square"))
  (if (Utilities|Array|ContainsItem (Variables|Default|GetOrientationOrder) CurrentOrientation)
    (Variables|Default|GetbForwardOrder
      (:then
        (LocalVariables|SetNextItemIterator 1))
      (:else
        (LocalVariables|SetNextItemIterator -1)))
    (return (Utilities|Array|Get(acopy) (Variables|Default|GetOrientationOrder) (Math|Integer|%25%28Integer%29 (+ (Math|Integer|%25%28Integer%29 (+ (Utilities|Array|FindItem (Variables|Default|GetOrientationOrder) (Variables|GetCurrentOrientation)) (LocalVariables|GetNextItemIterator)) (Utilities|Array|Length (Variables|Default|GetOrientationOrder))) (Utilities|Array|Length (Variables|Default|GetOrientationOrder))) (Utilities|Array|Length (Variables|Default|GetOrientationOrder)))))
    (else
      (return (Utilities|Array|Get(acopy) (Variables|Default|GetOrientationOrder))))))
