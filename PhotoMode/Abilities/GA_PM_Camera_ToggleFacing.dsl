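; Photo mode gameplay ability that toggles the camera facing state between selfie and back.
; On activation it checks the owner's ability system for the Selfie facing tag, passes the opposite pair to HandleFacingTags, then ends immediately.
; HandleFacingTags removes and adds replicated loose gameplay tags on the avatar, so the ability changes only tags, not the camera.
; A GetPlayerCamera helper is defined but unused in this graph.
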(event Ability|EventActivateAbility
  (Utilities|FlowControl|Sequence
    (:then_0
      (if (GameplayTags|HasMatchingGameplayTag (Ability|GetAbilitySystemComponentfromActorInfo) "(TagName=\"Ability.PhotoMode.Camera.Facing.Selfie\")")
        (CallFunction|HandleFacingTags "(GameplayTags=((TagName=\"Ability.PhotoMode.Camera.Facing.Back\")))" "(GameplayTags=((TagName=\"Ability.PhotoMode.Camera.Facing.Selfie\")))")
        (else
          (CallFunction|HandleFacingTags "(GameplayTags=((TagName=\"Ability.PhotoMode.Camera.Facing.Selfie\")))" "(GameplayTags=((TagName=\"Ability.PhotoMode.Camera.Facing.Back\")))"))))
    (:then_1
      (Ability|EndAbility))))

(fn GetPlayerCamera
  (bind _returnvalue (CallFunction|GetAvatarasFortPlayerPawn))
  (bind _playercameramanager (Variables|/Script/Engine.PlayerController|GetPlayerCameraManager (Fortnite|GetFortPlayerControllerfromActor _returnvalue))
    (:then
      (bind _asfort_player_camera_base (Utilities|Casting|CastToFortPlayerCameraBase _playercameramanager)
        (:then
          (return _asfort_player_camera_base _returnvalue))
        (:CastFailed
          (Ability|EndAbility)
          (return 0 0))))
    (:else
      (return 0 0))))

(fn HandleFacingTags
  (params (AddedTags :type Gameplay%20Tag%20Container%20Structure) (RemovedTags :type Gameplay%20Tag%20Container%20Structure))
  (Ability|GameplayEffect|RemoveLooseGameplayTags (Ability|GetAvatarActorfromActorInfo) (Variables|GetRemovedTags) true)
  (Ability|GameplayEffect|AddLooseGameplayTags (Ability|GetAvatarActorfromActorInfo) (Variables|GetAddedTags) true))
