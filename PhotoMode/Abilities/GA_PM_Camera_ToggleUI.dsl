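; Photo mode gameplay ability that signals a UI visibility toggle on the local client.
; On activation without authority, it sends a CustomVerb.PhotoMode.UIVisibility gameplay event built with the avatar Fort player pawn, then ends the ability.
; With authority, it only prints a warning against firing on the server, under the name GA_Camera_ToggleOrientation, and never calls EndAbility.
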(event Ability|EventActivateAbility
  (Ability|HasAuthority
    (:True
      (Development|PrintString "Warning: GA_Camera_ToggleOrientation should not be fired in Server"))
    (:False
      (Ability|SendGameplayEvent 0 "(TagName=\"CustomVerb.PhotoMode.UIVisibility\")" (Utilities|Struct|MakeGameplayEventData "(TagName=\"\")" (CallFunction|GetAvatarasFortPlayerPawn) (CallFunction|GetAvatarasFortPlayerPawn)))
      (Ability|EndAbility))))
