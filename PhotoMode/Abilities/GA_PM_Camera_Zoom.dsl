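; Photo Mode gameplay ability that turns camera zoom on while active and off when it ends.
; Activation resumes a cached zoom camera modifier, first finding it by class on the Fortnite player camera if unset, and prints an error if the cast fails.
; Ending pauses that modifier rather than removing it.
; Both events broadcast a zoom UI message through the global GameplayEventRouter to the avatar pawn's controller.
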(event Ability|EventActivateAbility
  (GameplayEventRouter|BroadcastEvent (GameplayEventRouter|GetGlobalEventRouter) (Utilities|Struct|MakeSPMZoomUIMessage true) (Pawn|GetController (CallFunction|GetAvatarasFortPlayerPawn)))
  (bind _camerazoommodifier (Variables|Default|GetCameraZoomModifier)
    (:then
      (Class|CMPMCameraZoomBase|ResumeZoom _camerazoommodifier))
    (:else
      (bind _player_camera_base (Default|GetPlayerCamera))
      (bind _returnvalue (Game|Player|FindCameraModifierbyClass _player_camera_base (Variables|Default|GetCameraModifier)))
      (bind _ascm_pm_camera_zoom_base (Utilities|Casting|CastToCM_PM_CameraZoom_Base _returnvalue)
        (:then
          (bind _output_get (Variables|Default|SetCameraZoomModifier _ascm_pm_camera_zoom_base))
          (Class|CMPMCameraZoomBase|ResumeZoom _output_get))
        (:CastFailed
          (Development|PrintText (Utilities|Text|FormatText "NSLOCTEXT(\"[63D3E788357BE9525B7AA6BF9AE8C9FD]\", \"A1C6979A45D5B0546D96D6B7E5C2ECE9\", \"Error: Camera mode not found in Ability {a}\")" self) true true "(R=1.000000,G=0.000000,B=0.065489,A=1.000000)" 5.0))))))

(event Ability|EventOnEndAbility
  (params (bWasCancelled :type Boolean))
  (GameplayEventRouter|BroadcastEvent (GameplayEventRouter|GetGlobalEventRouter) (Utilities|Struct|MakeSPMZoomUIMessage false) (Pawn|GetController (CallFunction|GetAvatarasFortPlayerPawn)))
  (bind _camerazoommodifier (Variables|Default|GetCameraZoomModifier)
    (:then
      (Class|CMPMCameraZoomBase|PauseZoom _camerazoommodifier))
    (:else)))

(fn GetPlayerCamera
  (bind _returnvalue (CallFunction|GetAvatarasFortPlayerPawn))
  (bind _playercameramanager (Variables|/Script/Engine.PlayerController|GetPlayerCameraManager (Fortnite|GetFortPlayerControllerfromActor _returnvalue))
    (:then
      (bind _asfort_player_camera_base (Utilities|Casting|CastToFortPlayerCameraBase _playercameramanager)
        (:then
          (return _asfort_player_camera_base _returnvalue))
        (:CastFailed
          (Ability|EndAbility)
          (return 0 0))))
    (:else
      (return 0 0))))
