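; GA_PM_Camera_ZoomOut is a photo mode gameplay ability, named for zooming the camera out.
; Its ActivateAbility event only calls the parent implementation, so the graph adds no zoom logic of its own.
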(event Ability|EventActivateAbility
  (|CallParentFunction))
